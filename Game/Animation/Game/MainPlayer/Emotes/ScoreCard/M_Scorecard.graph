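# Scorecard emote material that draws a one- or two-digit score in one color on a card of another, with a box-mask edge tint.
# Digit-Tens and Digit-Ones shift second-UV lookups along a number atlas, whose blue channel masks each digit's region.
# UseSingleDigit blends to a single ones-digit layout.
# High-contrast noise and HeightLerp break up the numeral shapes, and vertex color blue limits where they appear.
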Material /Game/Animation/Game/MainPlayer/Emotes/ScoreCard/M_Scorecard
graph {
 properties {pixelDepthOffsetMode=PDOM_Legacy, usedShadingModels=Default Lit}
 nodes [
  Constant() {r=0.2}
  Constant() {r=0.6}
  n1 = MaterialFunctionCall!(A=TextureCoordinate(), B=Constant2Vector() {g=0.5, r=0.5}, Bounds=Constant() {r=-0.1}, Edge Falloff=Constant() {r=1}) {materialFunction=/Engine/Functions/Engine_MaterialFunctions01/Gradient/BoxMask-2D}
  n2 = Saturate(None=Add(A=OneMinus(None=@n1.Result), B=Constant() {r=-0.2}))
  n3 = Constant() {r=0.0938}
  n4 = Constant() {r=-0.04}
  n5 = Constant() {r=-0.74}
  n6 = TextureCoordinate() {coordinateIndex=1}
  n7 = Constant() {r=1.2}
  n8 = MaterialFunctionCall(Float2=Add(
    A=MaterialFunctionCall(X=@n4, Y=@n5) {materialFunction=&MakeFloat2}.Result
    B=MaterialFunctionCall(UVs=@n6, Texture Scale=@n7) {materialFunction=&ScaleUVsByCenter}.UVs
   )) {materialFunction=&BreakOutFloat2Components}
  n9 = Constant() {r=0.2}
  n10 = Multiply(A=@n8.R, B=@n9)
  n11 = Add(A=Multiply(A=ScalarParameter() {defaultValue=1, parameterName=Digit-Tens}, B=@n3), B=Add(A=Constant() {r=0.03}, B=@n10))
  n12 = Add(A=Constant() {r=0.4}, B=Multiply(A=@n8.G, B=@n9))
  n13 = Add(A=Constant() {r=0.05}, B=@n10)
  n14 = Add(A=Constant() {r=-0.015}, B=@n10)
  n15 = ScalarParameter() {defaultValue=1, parameterName=UseSingleDigit}
  n16 = TextureSample(UVs=LinearInterpolate(
    A=MaterialFunctionCall(X=@n13, Y=@n12) {materialFunction=&MakeFloat2}.Result
    B=MaterialFunctionCall(X=@n14, Y=@n12) {materialFunction=&MakeFloat2}.Result
    Alpha=@n15
   )) {samplerType=SAMPLERTYPE_LinearColor, texture=/Game/Accessories/FORT_Backpacks/M_MED_Robot_Backpack/Textures/T_SimpleNumbers_Square}
  n17 = Multiply(
   A=TextureSample(UVs=MaterialFunctionCall(X=@n11, Y=@n12) {materialFunction=&MakeFloat2}.Result) {samplerType=SAMPLERTYPE_LinearColor, texture=&T_SimpleNumbers_Square}.R
   B=@n16.B
  )
  n18 = Multiply(A=ScalarParameter() {defaultValue=1, parameterName=Digit-Ones}, B=@n3)
  n19 = Add(A=@n18, B=Add(A=Constant() {r=-0.04}, B=@n10))
  n20 = Add(A=@n18, B=@n14)
  n21 = TextureSample(UVs=LinearInterpolate(
    A=MaterialFunctionCall(X=@n19, Y=@n12) {materialFunction=&MakeFloat2}.Result
    B=MaterialFunctionCall(X=@n20, Y=@n12) {materialFunction=&MakeFloat2}.Result
    Alpha=@n15
   )) {samplerType=SAMPLERTYPE_LinearColor, texture=&T_SimpleNumbers_Square}
  n22 = Multiply(A=@n21.R, B=OneMinus(None=@n16.B))
  n23 = Power(Base=LinearInterpolate(A=Add(A=@n17, B=@n22), B=Multiply(A=@n21.R, B=@n16.B), Alpha=@n15), Exp=Constant() {r=2})
  n24 = Saturate(None=LinearInterpolate(A=Constant() {r=-1}, B=Constant() {r=3}, Alpha=@n23))
  n25 = TextureCoordinate()
  n26 = Multiply(A=@n25, B=Constant() {r=6})
  n27 = Constant() {r=-5}
  n28 = Constant() {r=5}
  n29 = TextureSample(UVs=Multiply(A=@n25, B=Constant() {r=2.5})) {samplerType=SAMPLERTYPE_LinearColor, texture=&HighContrastNoise}
  n30 = Multiply(
   A=MaterialFunctionCall(
    Transition Phase=@n23
    Height Texture=TextureSample(UVs=@n26) {samplerType=SAMPLERTYPE_LinearColor, texture=&HighContrastNoise}.R
   ) {materialFunction=&HeightLerp}.Lerp Alpha No Contrast
   B=Saturate(None=LinearInterpolate(A=@n27, B=@n28, Alpha=@n29.G))
  )
  n31 = Multiply(A=VertexColor().B, B=Saturate(None=Add(A=@n24, B=@n30)))
  LinearInterpolate(
   A=VectorParameter() {parameterName=Color, defaultValue=<rgba 0.8 0.8 0.8 1>}.RGB
   B=VectorParameter() {parameterName=Color_1, defaultValue=<rgba 0.04 0.04 0.04 1>}.RGB
   Alpha=Saturate(None=Add(A=@n2, B=@n31))
  )
 ]
}
